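# Material function for the SuperNovaTaro pickaxe that sets Emissive Color and Opacity on incoming material attributes, producing an animated starfield.
# Three colored star layers scroll in polar UVs built around a center, each modulated by panning noise.
# A three-color background uses Fresnel and normal masks, alongside a mask texture and a local-position sphere mask.
# Time of day scales the emissive, and a second-UV gradient shapes opacity.
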MaterialFunction /BRCosmetics/Weapons/FORT_Melee/Pickaxe_SuperNovaTaro/Materials/MF_SuperNovaTaro_Pickaxe
graph {nodes [
  n1 = MaterialFunctionCall(
   Value_Morning=ScalarParameter() {defaultValue=0.8, parameterName=MorningStrength}
   Value_Day=Constant() {r=1}
   Value_Evening=ScalarParameter() {defaultValue=0.25, parameterName=EveningStrength}
   Value_Night=ScalarParameter() {defaultValue=0.25, parameterName=NightStrength}
  ) {materialFunction=&Scalar_Time_of_Day}
  n2 = NamedRerouteUsage()
  FunctionOutput(None=SetMaterialAttributes!(
    MaterialAttributes=FunctionInput() {inputType=FunctionInput_MaterialAttributes}
    EmissiveColor=Multiply(A=@n1.Result, B=@n2)
    Opacity=NamedRerouteUsage()
   ) {attributeSetTypes [#1, #2]})
  n3 = Constant()
  n4 = Add(A=Multiply(A=Time(), B=ScalarParameter() {defaultValue=0.36, parameterName=UV_Speed_03})) {constB=0.214}
  NamedRerouteDeclaration(None=Add(
    A=NamedRerouteUsage()
    B=MaterialFunctionCall(X=@n3, Y=@n4) {materialFunction=&MakeFloat2}.Result
   )) {name=PolarUV_03, nodeColor=<rgba 1 0.6797 0 1>}
  NamedRerouteDeclaration(None=ScalarParameter() {defaultValue=1, parameterName=NoiseMaskStrength}) {name=NoiseMaskStrength, nodeColor=<rgba 0.875 1 0 1>}
  n5 = Multiply(
   A=VectorParameter() {parameterName=Stars01_Color, defaultValue=<rgba 0.9885 0.1927 1 1>}.RGB
   B=ScalarParameter() {defaultValue=5, parameterName=Stars01_Strength}
  )
  n6 = TextureCoordinate()
  n7 = Multiply(
   A=@n5
   B=TextureSample(UVs=Panner(Coordinate=@n6) {speedX=-0.1, speedY=-0.1}) {samplerType=SAMPLERTYPE_Masks, texture=/Game/Characters/Player/Female/Medium/Bodies/F_MED_SuperNovaTaro/Textures/T_Stars_03}.G
  )
  n8 = NamedRerouteUsage()
  n9 = TextureSample(UVs=NamedRerouteUsage()) {samplerType=SAMPLERTYPE_Masks, texture=&T_Stars_03}
  n10 = NamedRerouteDeclaration(None=@n9.G) {name=NoiseMask, nodeColor=<rgba 0 1 0.2969 1>}
  NamedRerouteDeclaration(None=Multiply(A=LinearInterpolate(A=@n5, B=@n7, Alpha=@n8), B=Multiply(A=@n9.R, B=@n10))) {name=Stars_01, nodeColor=<rgba 1 0 0.4688 1>}
  n11 = Multiply(
   A=VectorParameter() {parameterName=Stars02_Color, defaultValue=<rgba 1 0 0.6047 1>}.RGB
   B=ScalarParameter() {defaultValue=15, parameterName=Stars02_Strength}
  )
  n12 = TextureCoordinate()
  n13 = Multiply(
   A=@n11
   B=TextureSample(UVs=Panner(Coordinate=@n12) {speedX=0.1, speedY=-0.11}) {samplerType=SAMPLERTYPE_Masks, texture=&T_Stars_03}.G
  )
  n14 = NamedRerouteUsage()
  n15 = TextureSample(UVs=NamedRerouteUsage()) {samplerType=SAMPLERTYPE_Masks, texture=&T_Stars_03}
  NamedRerouteDeclaration(None=Multiply(A=LinearInterpolate(A=@n11, B=@n13, Alpha=@n14), B=Multiply(A=@n15.R, B=@n15.R))) {name=Stars_02, nodeColor=<rgba 0 1 0.2266 1>}
  n16 = Multiply(
   A=VectorParameter() {parameterName=Stars03_Color, defaultValue=<rgba 0.1798 0.0307 1 1>}.RGB
   B=ScalarParameter() {defaultValue=10, parameterName=Stars03_Strength}
  )
  n17 = Multiply(A=TextureCoordinate()) {constB=0.765}
  n18 = Multiply(
   A=@n16
   B=TextureSample(UVs=Panner(Coordinate=@n17) {speedX=-0.06, speedY=-0.04}) {samplerType=SAMPLERTYPE_Masks, texture=&T_Stars_03}.G
  )
  n19 = NamedRerouteUsage()
  n20 = TextureSample(UVs=NamedRerouteUsage()) {samplerType=SAMPLERTYPE_Masks, texture=&T_Stars_03}
  NamedRerouteDeclaration(None=Multiply(A=LinearInterpolate(A=@n16, B=@n18, Alpha=@n19), B=Multiply(A=@n20.G, B=@n20.G))) {name=Stars_03, nodeColor=<rgba 0.2188 1 0 1>}
  n21 = Add(A=Add(A=NamedRerouteUsage(), B=NamedRerouteUsage()), B=NamedRerouteUsage())
  n22 = OneMinus(None=Multiply(A=NamedRerouteUsage(), B=ScalarParameter() {defaultValue=0.5, parameterName=StarsTransparentFade}))
  NamedRerouteDeclaration(None=Multiply(A=Multiply(A=@n21, B=@n22), B=ScalarParameter() {defaultValue=0.8, parameterName=StarsGlobalStrength})) {name=Stars, nodeColor=<rgba 0 1 0.8125 1>}
  n23 = VectorParameter() {parameterName=BG_Color_01, defaultValue=<rgba 0.01 0.01 0.01 1>}
  n24 = VectorParameter() {parameterName=BG_Color_02, defaultValue=<rgba 0.1406 0 0.135 1>}
  n25 = NamedRerouteUsage()
  n26 = ScalarParameter() {defaultValue=1, parameterName=BGColor02_MaskStrength}
  n27 = Fresnel(ExponentIn=ScalarParameter() {defaultValue=0.3, parameterName=FresnelExponent}) {baseReflectFraction=0}
  n28 = Power(
   Base=OneMinus(None=NamedRerouteDeclaration(None=@n27) {name=Fresnel, nodeColor=<rgba 0 0.7656 1 1>})
   Exp=ScalarParameter() {defaultValue=3, parameterName=FresnelContrast}
  )
  n29 = ScalarParameter() {defaultValue=5, parameterName=FresnelStrength}
  n30 = Multiply(
   A=Multiply(A=Multiply(A=@n25, B=@n26), B=NamedRerouteUsage())
   B=NamedRerouteDeclaration(None=Multiply(A=@n28, B=@n29)) {name=FresnelInvert, nodeColor=<rgba 0.6641 1 0 1>}
  )
  n31 = NamedRerouteUsage()
  n32 = Transform(None=Constant3Vector() {constant=<rgba 0 0 1 0>}) {transformType=TRANSFORM_Local}
  n33 = MaterialFunctionCall(
   Input=Abs(None=ComponentMask(None=@n32) {g=true})
   Input Low=ScalarParameter() {defaultValue=0.7, parameterName=NormalMaskMin}
   Input High=Constant() {r=1}
  ) {materialFunction=&RemapValue_0-1}
  n34 = ScalarParameter() {defaultValue=3, parameterName=BGColor03_MaskContrast}
  n35 = Multiply(
   A=NamedRerouteDeclaration(None=Power(Base=@n33.Result, Exp=@n34)) {name=NormalMask, nodeColor=<rgba 1 0 0.9141 1>}
   B=OneMinus(None=NamedRerouteUsage())
  )
  NamedRerouteDeclaration(None=LinearInterpolate(
    A=LinearInterpolate(A=@n23.RGB, B=@n24.RGB, Alpha=@n30)
    B=VectorParameter() {parameterName=BG_Color_03, defaultValue=<rgba 0.0329 0.0086 0.0521 1>}.RGB
    Alpha=Multiply(A=@n31, B=@n35)
   )) {name=BGColor, nodeColor=<rgba 0 0.8828 1 1>}
  n36 = TextureSample() {texture=/BRCosmetics/Weapons/FORT_Melee/Pickaxe_SuperNovaTaro/Textures/T_supernovataro_mask}
  NamedRerouteDeclaration(None=@n36.G) {name=Mask_Green, nodeColor=<rgba 1 0 0.9609 1>}
  n37 = Constant()
  n38 = Multiply(A=Time(), B=ScalarParameter() {defaultValue=0.3, parameterName=UV_Speed_01})
  NamedRerouteDeclaration(None=Add(
    A=NamedRerouteUsage()
    B=MaterialFunctionCall(X=@n37, Y=@n38) {materialFunction=&MakeFloat2}.Result
   )) {name=PolarUV_01, nodeColor=<rgba 1 0.6797 0 1>}
  NamedRerouteDeclaration(None=@n36.B) {name=Mask_Blue, nodeColor=<rgba 1 0 0.4219 1>}
  NamedRerouteDeclaration(None=@n36.R) {name=Mask_Red, nodeColor=<rgba 0 1 0.2969 1>}
  n39 = ScalarParameter() {parameterName=OpacityUVMin}
  n40 = ScalarParameter() {defaultValue=0.7, parameterName=OpacityUVMax}
  n41 = ComponentMask(None=TextureCoordinate() {coordinateIndex=1}) {g=true}
  n42 = NamedRerouteUsage()
  n43 = ScalarParameter() {defaultValue=1.15, parameterName=OpacityMaskStrength}
  NamedRerouteDeclaration(None=Multiply(A=InverseLinearInterpolate(A=@n39, B=@n40, Value=@n41), B=Multiply(A=@n42, B=@n43))) {name=UV_Opacity, nodeColor=<rgba 0 1 0.25 1>}
  n44 = ComponentMask(None=VectorParameter() {parameterName=UVDistort, defaultValue=<rgba 0 1.2 0 1>}.RGB) {g=true, r=true}
  n45 = NamedRerouteUsage()
  n46 = NamedRerouteUsage()
  n47 = ScalarParameter() {defaultValue=10, parameterName=UVDivisor}
  n48 = NamedRerouteDeclaration(None=Add(A=Divide(A=@n46, B=@n47)) {constB=0.5}) {name=Center, nodeColor=<rgba 0.1719 1 0 1>}
  n49 = ComponentMask(None=MaterialFunctionCall!(Coordinate=OneMinus(None=@n48)) {materialFunction=/Game/Effects/Niagara/MasterMaterials/Functions/UVs/PolarCoordinates}.Result) {g=true, r=true}
  n50 = ComponentMask(None=VectorParameter() {parameterName=UVTiling, defaultValue=<rgba 3 1 0 1>}.RGB) {g=true, r=true}
  NamedRerouteDeclaration(None=Add(A=Multiply(A=@n44, B=@n45), B=Multiply(A=@n49, B=@n50))) {name=PolarCoords, nodeColor=<rgba 0 0.4375 1 1>}
  n51 = Constant()
  n52 = Multiply(A=Time(), B=ScalarParameter() {defaultValue=0.525, parameterName=UV_Speed_02})
  NamedRerouteDeclaration(None=Add(
    A=NamedRerouteUsage()
    B=MaterialFunctionCall(X=@n51, Y=@n52) {materialFunction=&MakeFloat2}.Result
   )) {name=PolarUV_02, nodeColor=<rgba 1 0.6797 0 1>}
  n53 = NamedRerouteUsage()
  n54 = NamedRerouteUsage()
  n55 = LinearInterpolate(A=Multiply(A=@n54, B=ScalarParameter() {defaultValue=0.5, parameterName=StarsSphereMaskFade}), B=@n54, Alpha=NamedRerouteUsage())
  NamedRerouteDeclaration(None=Add(A=Multiply(A=@n53, B=@n55), B=NamedRerouteUsage())) {name=Color, nodeColor=<rgba 0.2422 0 1 1>}
  n56 = Add(
   A=ComponentMask(None=MaterialFunctionCall() {materialFunction=&LocalPosition}.Local Position) {b=true, r=true}
   B=ComponentMask(None=VectorParameter() {parameterName=HoleCenterOffset, defaultValue=<rgba 3.5 0 2.5 1>}.RGB) {b=true, r=true}
  )
  n57 = Add(A=Length(None=NamedRerouteDeclaration(None=@n56) {name=LocalPosXZ, nodeColor=<rgba 1 0.8672 0 1>}) {is=class MaterialExpressionLength}) {constB=-10}
  n58 = NamedRerouteDeclaration(None=Saturate(None=Divide(A=@n57) {constA=0, constB=30})) {name=SphereMask_01, nodeColor=<rgba 0.2422 0 1 1>}
  NamedRerouteDeclaration(None=Power(Base=OneMinus(None=@n58))) {name=SphereMaskInvert, nodeColor=<rgba 0 0.5078 1 1>}
 ]}
