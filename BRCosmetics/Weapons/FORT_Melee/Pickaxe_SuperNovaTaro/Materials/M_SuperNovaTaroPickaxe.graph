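# Translucent, unlit material for the SuperNovaTaro pickaxe, built on material attributes.
# The output chain runs CharacterParent 2023 PreFX through MF_SuperNovaTaro_Pickaxe into PostFX.
# Side networks declare named reroutes for hole-centered local XZ position, an inverted radial mask, distorted polar coordinates, three time-scrolled polar UVs and RGB texture masks.
# These side networks show no connection to that output chain, so the visible effect comes from the material functions.
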Material /BRCosmetics/Weapons/FORT_Melee/Pickaxe_SuperNovaTaro/Materials/M_SuperNovaTaroPickaxe
graph {
 properties {
  bUseMaterialAttributes=true
  bUsedWithMeshDeformer=true
  bUsedWithSkeletalMesh=true
  blendMode=BLEND_Translucent
  pixelDepthOffsetMode=PDOM_Legacy
  shadingModel=MSM_Unlit
  usedShadingModels=Unlit
 }
 nodes [
  n1 = MaterialFunctionCall() {materialFunction=&MF_CharacterParent_2023_PreFX}
  MaterialFunctionCall(
   Material=MaterialFunctionCall!(In=@n1.Result) {materialFunction=&/MF_SuperNovaTaro_Pickaxe}.Result
   M=@n1.M
   Pre Skinned Local Position=@n1.M
   Distance Blend=@n1.M
   Fresnel=@n1.M
  ) {materialFunction=&MF_CharacterParent_2023_PostFX}
  n2 = Constant()
  n3 = Add(A=Multiply(A=Time(), B=ScalarParameter() {defaultValue=0.36, parameterName=UV_Speed_03})) {constB=0.214}
  NamedRerouteDeclaration(None=Add(
    A=NamedRerouteUsage()
    B=MaterialFunctionCall(X=@n2, Y=@n3) {materialFunction=&MakeFloat2}.Result
   )) {name=PolarUV_03, nodeColor=<rgba 1 0.6797 0 1>}
  n4 = TextureSample() {texture=/BRCosmetics/Weapons/FORT_Melee/Pickaxe_SuperNovaTaro/Textures/T_supernovataro_mask}
  NamedRerouteDeclaration(None=@n4.G) {name=Mask_Green, nodeColor=<rgba 1 0 0.9609 1>}
  n5 = Constant()
  n6 = Multiply(A=Time(), B=ScalarParameter() {defaultValue=0.3, parameterName=UV_Speed_01})
  NamedRerouteDeclaration(None=Add(
    A=NamedRerouteUsage()
    B=MaterialFunctionCall(X=@n5, Y=@n6) {materialFunction=&MakeFloat2}.Result
   )) {name=PolarUV_01, nodeColor=<rgba 1 0.6797 0 1>}
  NamedRerouteDeclaration(None=@n4.B) {name=Mask_Blue, nodeColor=<rgba 1 0 0.4219 1>}
  NamedRerouteDeclaration(None=@n4.R) {name=Mask_Red, nodeColor=<rgba 0 1 0.2969 1>}
  n7 = ComponentMask(None=VectorParameter() {parameterName=UVDistort, defaultValue=<rgba 0 1.2 0 1>}.RGB) {g=true, r=true}
  n8 = NamedRerouteUsage()
  n9 = NamedRerouteUsage()
  n10 = ScalarParameter() {defaultValue=10, parameterName=UVDivisor}
  n11 = NamedRerouteDeclaration(None=Add(A=Divide(A=@n9, B=@n10)) {constB=0.5}) {name=Center, nodeColor=<rgba 0.1719 1 0 1>}
  n12 = ComponentMask(None=MaterialFunctionCall!(Coordinate=OneMinus(None=@n11)) {materialFunction=/Game/Effects/Niagara/MasterMaterials/Functions/UVs/PolarCoordinates}.Result) {g=true, r=true}
  n13 = ComponentMask(None=VectorParameter() {parameterName=UVTiling, defaultValue=<rgba 3 1 0 1>}.RGB) {g=true, r=true}
  NamedRerouteDeclaration(None=Add(A=Multiply(A=@n7, B=@n8), B=Multiply(A=@n12, B=@n13))) {name=PolarCoords, nodeColor=<rgba 0 0.4375 1 1>}
  n14 = Constant()
  n15 = Multiply(A=Time(), B=ScalarParameter() {defaultValue=0.525, parameterName=UV_Speed_02})
  NamedRerouteDeclaration(None=Add(
    A=NamedRerouteUsage()
    B=MaterialFunctionCall(X=@n14, Y=@n15) {materialFunction=&MakeFloat2}.Result
   )) {name=PolarUV_02, nodeColor=<rgba 1 0.6797 0 1>}
  n16 = Add(
   A=ComponentMask(None=MaterialFunctionCall() {materialFunction=&LocalPosition}.Local Position) {b=true, r=true}
   B=ComponentMask(None=VectorParameter() {parameterName=HoleCenterOffset, defaultValue=<rgba 3.5 0 2.5 1>}.RGB) {b=true, r=true}
  )
  n17 = Add(A=Length(None=NamedRerouteDeclaration(None=@n16) {name=LocalPosXZ, nodeColor=<rgba 1 0.8672 0 1>}) {is=class MaterialExpressionLength}) {constB=-10}
  n18 = NamedRerouteDeclaration(None=Saturate(None=Divide(A=@n17) {constA=0, constB=30})) {name=SphereMask_01, nodeColor=<rgba 0.2422 0 1 1>}
  NamedRerouteDeclaration(None=Power(Base=OneMinus(None=@n18))) {name=SphereMaskInvert, nodeColor=<rgba 0 0.5078 1 1>}
 ]
}
